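# Sound cue for running footstep foley of plastic claws on Mimosa biped cosmetic companions, classed as creature pet audio.
# A quality-level switch selects a random pool of ten waves for the first two levels or just the first four for the other two.
# A modulator varies pitch and volume per play, with shared companion foley attenuation and footstep concurrency governing playback.
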SoundCue /CosmeticCompanions/Assets/Audio_Shared/Foley/Biped/Claws/Claws_Plastic/SC_MImosa_Biped_Claws_Plastic_Run
root=Modulator_0
nodes [
 {id=Modulator_0, children [QualityLevel_0], properties {pitchMax=1.3, pitchMin=0.9, volumeMax=1, volumeMin=0.8}}
 {id=QualityLevel_0, children [Random_0,Random_0,Random_1,Random_1]}
 {
  id=Random_0
  children [
   WavePlayer_14
   WavePlayer_15
   WavePlayer_16
   WavePlayer_17
   WavePlayer_18
   WavePlayer_19
   WavePlayer_20
   WavePlayer_21
   WavePlayer_22
   WavePlayer_23
  ]
  properties {weights [1,1,1,1,1,1,1,1,1,1]}
 }
 {id=Random_1, children [WavePlayer_24, WavePlayer_25, WavePlayer_26, WavePlayer_27], properties {weights [1,1,1,1]}}
 {
  id=WavePlayer_14
  properties {soundWaveAssetPtr=/CosmeticCompanions/Assets/Audio_Shared/Foley/Biped/Claws/Claws_Plastic/Wavs/Mimosa_Biped_Claws_Plastic_Run_01}
 }
 {
  id=WavePlayer_15
  properties {soundWaveAssetPtr=/CosmeticCompanions/Assets/Audio_Shared/Foley/Biped/Claws/Claws_Plastic/Wavs/Mimosa_Biped_Claws_Plastic_Run_02}
 }
 {
  id=WavePlayer_16
  properties {soundWaveAssetPtr=/CosmeticCompanions/Assets/Audio_Shared/Foley/Biped/Claws/Claws_Plastic/Wavs/Mimosa_Biped_Claws_Plastic_Run_03}
 }
 {
  id=WavePlayer_17
  properties {soundWaveAssetPtr=/CosmeticCompanions/Assets/Audio_Shared/Foley/Biped/Claws/Claws_Plastic/Wavs/Mimosa_Biped_Claws_Plastic_Run_04}
 }
 {
  id=WavePlayer_18
  properties {soundWaveAssetPtr=/CosmeticCompanions/Assets/Audio_Shared/Foley/Biped/Claws/Claws_Plastic/Wavs/Mimosa_Biped_Claws_Plastic_Run_05}
 }
 {
  id=WavePlayer_19
  properties {soundWaveAssetPtr=/CosmeticCompanions/Assets/Audio_Shared/Foley/Biped/Claws/Claws_Plastic/Wavs/Mimosa_Biped_Claws_Plastic_Run_06}
 }
 {
  id=WavePlayer_20
  properties {soundWaveAssetPtr=/CosmeticCompanions/Assets/Audio_Shared/Foley/Biped/Claws/Claws_Plastic/Wavs/Mimosa_Biped_Claws_Plastic_Run_07}
 }
 {
  id=WavePlayer_21
  properties {soundWaveAssetPtr=/CosmeticCompanions/Assets/Audio_Shared/Foley/Biped/Claws/Claws_Plastic/Wavs/Mimosa_Biped_Claws_Plastic_Run_08}
 }
 {
  id=WavePlayer_22
  properties {soundWaveAssetPtr=/CosmeticCompanions/Assets/Audio_Shared/Foley/Biped/Claws/Claws_Plastic/Wavs/Mimosa_Biped_Claws_Plastic_Run_09}
 }
 {
  id=WavePlayer_23
  properties {soundWaveAssetPtr=/CosmeticCompanions/Assets/Audio_Shared/Foley/Biped/Claws/Claws_Plastic/Wavs/Mimosa_Biped_Claws_Plastic_Run_10}
 }
 {
  id=WavePlayer_24
  properties {soundWaveAssetPtr=&Mimosa_Biped_Claws_Plastic_Run_01}
 }
 {
  id=WavePlayer_25
  properties {soundWaveAssetPtr=&Mimosa_Biped_Claws_Plastic_Run_02}
 }
 {
  id=WavePlayer_26
  properties {soundWaveAssetPtr=&Mimosa_Biped_Claws_Plastic_Run_03}
 }
 {
  id=WavePlayer_27
  properties {soundWaveAssetPtr=&Mimosa_Biped_Claws_Plastic_Run_04}
 }
]
properties {
 attenuationSettings=/Game/Sounds/Attenuation/AS_Companion_Foley
 maxDistance=1200
 pitchMultiplier=0.8
 priority=0
 soundClassObject=&Fort_Creatures_Pets
 volumeMultiplier=0.4
 concurrencySet [/CosmeticCompanions/Assets/Audio_Shared/Foley/Biped/CONC_Mimosa_Biped_Footstep]
}
