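# Sound cue for short plastic claw slide foley on biped cosmetic companions (Mimosa).
# A quality-level node sends its first two slots to a random pick of eight wave variations and its last two to a smaller pool of the first four.
# Playback is pitched up and quietened, using companion foley attenuation, the creature pets sound class and the shared biped footstep concurrency.
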SoundCue /CosmeticCompanions/Assets/Audio_Shared/Foley/Biped/Claws/Claws_Plastic/SC_MImosa_Biped_Claws_Plastic_SlideShort
root=QualityLevel_0
nodes [
 {id=QualityLevel_0, children [Random_0,Random_0,Random_1,Random_1]}
 {
  id=Random_0
  children [WavePlayer_11, WavePlayer_12, WavePlayer_13, WavePlayer_14, WavePlayer_15, WavePlayer_16, WavePlayer_19, WavePlayer_20]
  properties {weights [1,1,1,1,1,1,1,1]}
 }
 {id=Random_1, children [WavePlayer_21, WavePlayer_22, WavePlayer_23, WavePlayer_24], properties {weights [1,1,1,1]}}
 {
  id=WavePlayer_11
  properties {soundWaveAssetPtr=/CosmeticCompanions/Assets/Audio_Shared/Foley/Biped/Claws/Claws_Plastic/Wavs/Mimosa_Biped_Claws_Plastic_SlideShort_01}
 }
 {
  id=WavePlayer_12
  properties {soundWaveAssetPtr=/CosmeticCompanions/Assets/Audio_Shared/Foley/Biped/Claws/Claws_Plastic/Wavs/Mimosa_Biped_Claws_Plastic_SlideShort_02}
 }
 {
  id=WavePlayer_13
  properties {soundWaveAssetPtr=/CosmeticCompanions/Assets/Audio_Shared/Foley/Biped/Claws/Claws_Plastic/Wavs/Mimosa_Biped_Claws_Plastic_SlideShort_03}
 }
 {
  id=WavePlayer_14
  properties {soundWaveAssetPtr=/CosmeticCompanions/Assets/Audio_Shared/Foley/Biped/Claws/Claws_Plastic/Wavs/Mimosa_Biped_Claws_Plastic_SlideShort_04}
 }
 {
  id=WavePlayer_15
  properties {soundWaveAssetPtr=/CosmeticCompanions/Assets/Audio_Shared/Foley/Biped/Claws/Claws_Plastic/Wavs/Mimosa_Biped_Claws_Plastic_SlideShort_05}
 }
 {
  id=WavePlayer_16
  properties {soundWaveAssetPtr=/CosmeticCompanions/Assets/Audio_Shared/Foley/Biped/Claws/Claws_Plastic/Wavs/Mimosa_Biped_Claws_Plastic_SlideShort_06}
 }
 {
  id=WavePlayer_19
  properties {soundWaveAssetPtr=/CosmeticCompanions/Assets/Audio_Shared/Foley/Biped/Claws/Claws_Plastic/Wavs/Mimosa_Biped_Claws_Plastic_SlideShort_07}
 }
 {
  id=WavePlayer_20
  properties {soundWaveAssetPtr=/CosmeticCompanions/Assets/Audio_Shared/Foley/Biped/Claws/Claws_Plastic/Wavs/Mimosa_Biped_Claws_Plastic_SlideShort_08}
 }
 {
  id=WavePlayer_21
  properties {soundWaveAssetPtr=&Mimosa_Biped_Claws_Plastic_SlideShort_01}
 }
 {
  id=WavePlayer_22
  properties {soundWaveAssetPtr=&Mimosa_Biped_Claws_Plastic_SlideShort_02}
 }
 {
  id=WavePlayer_23
  properties {soundWaveAssetPtr=&Mimosa_Biped_Claws_Plastic_SlideShort_03}
 }
 {
  id=WavePlayer_24
  properties {soundWaveAssetPtr=&Mimosa_Biped_Claws_Plastic_SlideShort_04}
 }
]
properties {
 attenuationSettings=/Game/Sounds/Attenuation/AS_Companion_Foley
 maxDistance=1200
 pitchMultiplier=1.3
 priority=0
 soundClassObject=&Fort_Creatures_Pets
 volumeMultiplier=0.3
 concurrencySet [/CosmeticCompanions/Assets/Audio_Shared/Foley/Biped/CONC_Mimosa_Biped_Footstep]
}
